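# Homebase material function that bends geometry around the camera into a curved "small world" via World Position Offset.
# It scales horizontal camera distance by SmallWorldBendValue from the Homebase parameter collection to get a rotation angle.
# Vertices rotate about a horizontal axis perpendicular to their camera direction, pivoting at camera XY and a constant Z.
# FixRotateAboutAxisNormals supplies corrected tangent-space normals, output as separate RG and B.
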MaterialFunction /Game/Environments/Homebase/MaterialFunctions/MF_HB_SmallWorld
graph {
 properties {bExposeToLibrary=true}
 nodes [
  Constant() {desc=DistanceOffset / Bend Value, r=-410.893}
  n1 = WorldPosition()
  n2 = ComponentMask(None=CameraPositionWS()) {g=true, r=true}
  n3 = Distance(A=ComponentMask(None=@n1.XYZ) {g=true, r=true}, B=@n2)
  n4 = CollectionParameter() {
   collection=/Game/Environments/Homebase/Materials/HomeBaseParameterCollection
   desc=values from 0 to -0.005 are good
   parameterName=SmallWorldBendValue
  }
  Divide(A=@n3, B=@n4)
  n5 = AppendVector(A=@n2, B=Constant() {desc=WorldCenter Z})
  n6 = Normalize(VectorInput=CrossProduct(A=Subtract(A=@n1.XYZ, B=@n5), B=Constant3Vector() {constant=<rgba 0 0 1 0>}))
  n7 = Divide(A=Multiply(A=@n3, B=@n4)) {constB=360}
  FunctionOutput(None=RotateAboutAxis(NormalizedRotationAxis=@n6, RotationAngle=@n7, PivotPoint=@n5, Position=WorldPosition().XYZ)) {outputName=WorldPositionOffset, sortPriority=1}
  n8 = MaterialFunctionCall!(Rotation Axis=@n6, Rotation Angle=@n7, World Space Vertex Normals, Movement Mask) {materialFunction=&FixRotateAboutAxisNormals}
  FunctionOutput(None=@n8.New Tangent Space Vertex Normal RG) {outputName=NormalRG, sortPriority=3}
  FunctionOutput(None=@n8.New Tangent Space Vertex Normal B) {outputName=NormalB, sortPriority=5}
  StaticSwitch()
 ]
}
